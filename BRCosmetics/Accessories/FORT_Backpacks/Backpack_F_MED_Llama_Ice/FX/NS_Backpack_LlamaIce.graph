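# Niagara effect for the Llama Ice backpack cosmetic, producing a flickering blue glow.
# A single Light emitter burst-spawns a persistent particle with looping lifetime, drawn by both sprite and light renderers with a small camera offset.
# An engine-time Waveform drives sprite alpha and light color scale together, over constant base levels.
# User.bIsFrontEndPreview selects a brighter blue at spawn.
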NiagaraSystem /BRCosmetics/Accessories/FORT_Backpacks/Backpack_F_MED_Llama_Ice/FX/NS_Backpack_LlamaIce
dependencies {usedDataInterfaces [], usedDynamicInputs [], usedModules [&SystemState]}
emitters {Light {
  emitterUpdateScript [
   &EmitterState(Life Cycle Mode=::NewEnumerator1, Loop Behavior=::NewEnumerator1, Loop Duration Mode=::NewEnumerator1, UseLoopDelay=false)
   &SpawnBurst_Instantaneous(Use Spawn Probability=false, UseLoopCountLimit=false)
  ]
  particleSpawnScript [&InitializeParticle(
    Color=&MakeCustomLinearColorFromBool(Bool=$User.bIsFrontEndPreview, True Color=<rgba 0 0.6899 2 1>, False Color=<rgba 0 0.3449 1 1>)
    Lifetime=2
    Sprite Size Mode=::NewEnumerator3
    Uniform Sprite Size=55
    UsePositionOffset=false
    Write Material Random=true
   )]
  particleUpdateScript [
   &ParticleState(
    Kill Particles When Lifetime Has Elapsed=false
    Let Infinitely Lived Particles Die When Emitter Deactivates=true
    Loop Particles Lifetime=true
   )
   /Niagara/Noise/Waveform(
    EnableExponent1=false
    Global Phase:NiagaraFloat=$Engine.Time
    GlobalClampMax=false
    GlobalClampMin=false
    UseInput1=false
    UsePhase1=false
    "Waveform [1]":ENiagara_Waveforms=::NewEnumerator12
    "[1] Amplitude Min/Max":Vector2f=<0 0.2>
    "[1] Frequency"=5
   )
   &ScaleColor(Scale Alpha=&Add_Float(A=0.2, B=$Particles.Waveform.Waveform), ScaleA=true, ScaleRGB=true)
   &FN_Light_Attributes(
    Color=$Particles.Color
    Create Color Attribute=true
    Create Exponent Attribute=true
    Create Radius Attribute=true
    Exponent=7
    Radius=50
    Scale Color=&Add_Float(A=3, B=&Multiply_Float(A=$Particles.Waveform.Waveform, B=3))
   )
   &CameraOffset(Camera Offset Amount=7)
  ]
  rendererClasses [&NiagaraSpriteRendererProperties, &NiagaraLightRendererProperties]
  renderers [
   {rendererClass=&NiagaraSpriteRendererProperties}
   {rendererClass=&NiagaraLightRendererProperties, rendererIndex=1}
  ]
 }}
summary {
 systemName=NS_Backpack_LlamaIce
 emitters [{emitterName=Light, rendererClasses [&NiagaraSpriteRendererProperties, &NiagaraLightRendererProperties]}]
 userVariables [
  User.bIsFrontEnd {type=NiagaraBool}
  User.bIsFrontEndPreview {type=NiagaraBool}
 ]
}
